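# M_RocketBlast is a translucent deferred decal material for a rocket blast mark, stored with the Super V Lair environment props.
# The RocketBlast texture's RGB runs through CheapContrast, driven by an AlphaContrast parameter, then is multiplied by OpacityStrength to form a mask.
# Color, black by default, and Roughness are separate parameters, so instances can adjust the mark's tint, surface response and strength.
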Material /Game/Environments/Sets/Super_V_Lair/Props/decal/M_RocketBlast
graph {
 properties {bAutomaticallySetUsageInEditor=false, blendMode=BLEND_Translucent, materialDomain=MD_DeferredDecal, usedShadingModels=Default Lit}
 nodes [
  Multiply(
   A=MaterialFunctionCall(
    In=TextureSample() {texture=&/RocketBlast}.RGB
    Contrast=ScalarParameter() {parameterName=AlphaContrast}
   ) {materialFunction=&CheapContrast}.Result
   B=ScalarParameter() {defaultValue=1, parameterName=OpacityStrength}
  )
  ScalarParameter() {defaultValue=1, parameterName=Roughness}
  VectorParameter() {parameterName=Color, defaultValue=<rgba 0 0 0 1>}
 ]
}
